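# Niagara system for the toon-styled swing trail on the DairyString Wade pickaxe, with gold-orange and pink color variants.
# Two short-lived ribbon emitters follow anim-trail sockets: a wider LaserMainTrail with three renderers and a narrower, single-renderer SubTrail.
# Curves over particle and emitter age shape ribbon width, while system swing-fade and time-of-day values scale width, alpha and color.
# The main trail also writes a separate laser color and enlarged alternate width.
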NiagaraSystem /BRCosmetics/Weapons/FORT_Melee/Pickaxe_DairyString_Wade/FX/NS_Pickaxe_DairyString_Wade_TrailToon
dependencies {
 usedDataInterfaces [&NiagaraDataInterfaceCurve, &NiagaraDataInterfaceSocketReader, &NiagaraDataInterfaceColorCurve]
 usedDynamicInputs [&ColorFromCurve]
 usedModules [
  &self:SystemSpawnScript.SystemScriptSource.SystemScriptGraph.NiagaraNodeAssignment_1.SetVariables_#1
  &SystemState
  &FN_VariantManager
  &self:SystemSpawnScript.SystemScriptSource.SystemScriptGraph.NiagaraNodeAssignment_4.SetVariables_#2
  &self:SystemSpawnScript.SystemScriptSource.SystemScriptGraph.NiagaraNodeAssignment_0.SetVariables_#3
 ]
}
emitters {
 LaserMainTrail {
  emitterSpawnScript [&FN_RibbonIDOffsetUtility()]
  emitterUpdateScript [
   &EmitterState()
   &FN_SpawnAnimTrail()
  ]
  particleSpawnScript [
   &FN_InitializeAnimTrailBoneIndecies(First Bone or Socket Index=2, Second Bone or Socket Index=3)
   &FN_InitializeAnimTrail(
    AnimTrailWidth=1.485
    Color=&VariantColorSelector(Variant Amount=0, Original Color=<rgba 3.1 1.9633 0 0.8>, Variant 1 Color=<rgba 3.1 0.31 2.077 0.8>, Variant 2 Color=?)
    Lifetime=0.3
   )
  ]
  particleUpdateScript [
   &FN_UpdateAnimTrail() {module=FN_UpdateAnimTrail001}
   &ScaleColor(
    Scale Alpha=&Multiply_Float(A=$System.SwingGeneralFade, B=1)
    Scale RGB=&MakeVectorFromLinearColorRGB(Color=$System.TODColor)
    ScaleA=true
    ScaleRGB=true
   )
   &ScaleRibbonWidth(Ribbon Width Scale=&Multiply_Float(
     A=&FloatFromCurve(FloatCurve=DataInterface() {propertyValues {Curve {keys [
          {arriveTangent=6.5271, interpMode=RCIM_Cubic, leaveTangent=6.5271, tangentMode=RCTM_User, time=0, value=0}
          <at 0.075 1 RCIM_Cubic>
          {arriveTangentWeight=0.1917, interpMode=RCIM_Cubic, tangentMode=RCTM_User, time=0.2667, value=0}
         ]}}})
     B=&FloatFromCurve(FloatCurve=DataInterface() {propertyValues {Curve {keys [<at 0.085 0 RCIM_Cubic>,<at 0.2 1 RCIM_Cubic>]}}}, CurveIndex=$Emitter.Age)
    ))
   &FN_Sprite_Attributes(
    Color=&MakeLinearColorFromVectorAndFloat(
     "Vector (RGB)"=&MakeVectorFromLinearColorRGB(Color=&VariantColorSelector(Variant Amount=0, Original Color=<rgba 1 0.6333 0 1>, Variant 1 Color=<rgba 1.5 0 1.2 1>, Variant 2 Color=?))
     "Float (Alpha)"=&FloatFromCurve(FloatCurve=DataInterface() {propertyValues {Curve {keys [<at 0 1 RCIM_Cubic>,<at 0.5 0 RCIM_Cubic>]}}})
    )
    Create Color Attribute=true
    Write to Module Namespace=true
   )
   /SetVariables(Particles.LaserColor:LinearColor=&MakeLinearColorFromVectorAndFloat(
     "Vector (RGB)"=&MakeVectorFromLinearColorRGB(Color=&VariantColorSelector(Variant Amount=0, Original Color=<rgba 10 7.3508 0.65 1>, Variant 1 Color=<rgba 10 2 6.9617 1>, Variant 2 Color=?))
     "Float (Alpha)"=1
    )) {isSetParameters=true}
   &ScaleRibbonWidth(Ribbon Width Scale=$System.SwingGeneralFade) {module=ScaleRibbonWidth001}
   /SetVariables(Particles.RibbonWidthAlt:NiagaraFloat=&Multiply_Float(A=$Particles.RibbonWidth, B=1.8)) {isSetParameters=true}
  ]
  rendererClasses [&NiagaraRibbonRendererProperties]
  renderers [
   {rendererClass=&NiagaraRibbonRendererProperties}
   {rendererClass=&NiagaraRibbonRendererProperties, rendererIndex=1}
   {rendererClass=&NiagaraRibbonRendererProperties, rendererIndex=2}
  ]
 }
 SubTrail {
  emitterSpawnScript [&FN_RibbonIDOffsetUtility()]
  emitterUpdateScript [
   &EmitterState()
   &FN_SpawnAnimTrail()
  ]
  particleSpawnScript [
   &FN_InitializeAnimTrailBoneIndecies(First Bone or Socket Index=2, Second Bone or Socket Index=3)
   &FN_InitializeAnimTrail(
    AnimTrailWidth=0.72
    Color=&VariantColorSelector(Variant Amount=0, Original Color=<rgba 2.1 0.7 0 2>, Variant 1 Color=<rgba 2.1 0 1.12 2>, Variant 2 Color=?)
    Lifetime=0.3
   )
  ]
  particleUpdateScript [
   &FN_UpdateAnimTrail() {module=FN_UpdateAnimTrail001}
   &ScaleColor(
    Scale Alpha=&Multiply_Float(A=$System.SwingGeneralFade, B=1)
    Scale RGB=&MakeVectorFromLinearColorRGB(Color=$System.TODColor)
    ScaleA=true
    ScaleRGB=true
   )
   &ScaleRibbonWidth(Ribbon Width Scale=&Multiply_Float(
     A=&FloatFromCurve(FloatCurve=DataInterface() {propertyValues {Curve {keys [
          {arriveTangent=6.5271, interpMode=RCIM_Cubic, leaveTangent=6.5271, tangentMode=RCTM_User, time=0, value=0}
          <at 0.075 1 RCIM_Cubic>
          {arriveTangentWeight=0.1917, interpMode=RCIM_Cubic, tangentMode=RCTM_User, time=0.2667, value=0}
         ]}}})
     B=&FloatFromCurve(FloatCurve=DataInterface() {propertyValues {Curve {keys [<at 0.1 0 RCIM_Cubic>,<at 0.2 1 RCIM_Cubic>]}}}, CurveIndex=$Emitter.Age)
    ))
   &ScaleRibbonWidth(Ribbon Width Scale=$System.SwingGeneralFade) {module=ScaleRibbonWidth001}
  ]
  rendererClasses [&NiagaraRibbonRendererProperties]
  renderers [{rendererClass=&NiagaraRibbonRendererProperties}]
 }
}
summary {
 systemName=NS_Pickaxe_DairyString_Wade_TrailToon
 emitters [
  {emitterName=LaserMainTrail, rendererClasses [&NiagaraRibbonRendererProperties]}
  {emitterName=SubTrail, rendererClasses [&NiagaraRibbonRendererProperties]}
 ]
 userVariables [
  User.bIsFrontEnd {type=NiagaraBool}
  User.bIsFrontEndPreview {type=NiagaraBool}
  User.RibbonIDOffset {type=NiagaraInt32}
  User.Variant {type=NiagaraFloat, defaultValue=0, description=""}
 ]
}
