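# Animation Blueprint for the DairyString Wade pickaxe that adds secondary physics motion to its six dyn_banana bones.
# The cached input pose feeds three chains of root-relative Anim Dynamics nodes with angular springs, differing in tuning.
# Nested bool blends select one chain when bIsOnPawn is true, otherwise one of the other two by bIsInFrontEnd.
# The EventGraph holds only unconnected stub nodes.
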AnimBlueprint /BRCosmetics/Weapons/FORT_Melee/Pickaxe_DairyString_Wade/Meshes/DairyString_Wade_Axe_AnimBP
target_skeleton=&/DairyString_Wade_Axe_Skeleton
graphs [
 AnimGraph {nodes [
   n1 = Anim Dynamics - Bone: dyn_banana_c_01(ComponentPose=Local To Component(LocalPose=Use cached pose 'PX_dairy_01'()), Alpha=1, ExternalForce="0,0,0") {
    angularDampingOverride=1
    angularSpringConstant=11000
    bAngularSpring=true
    bOverrideAngularDamping=true
    bUseGravityOverride=true
    gravityScale=0.35
    simulationSpace=RootRelative
    boundBone {boneName=dyn_banana_c_01}
    componentAppliedLinearAccClamp=<2000 2000 2000>
    componentLinearAccScale=<0.25 0.14 0.1>
    componentLinearVelScale=<0.25 0.14 0.1>
    gravityOverride {z=-700}
    physicsBodyDefinitions [{
      boundBone {boneName=dyn_banana_c_01}
      boxExtents=<45 72 33>
      constraintSetup {angularLimitsMax=<0 19 2>, angularLimitsMin=<0 -26 -2>, angularTarget=<1 0 0>}
      localJointOffset=<1.6 0 0>
     }]
    simSpaceSettings {maxAngularAcceleration=25, maxAngularVelocity=2.5, simSpaceAngularAlpha=1}
   }
   n2 = Anim Dynamics - Bone: dyn_banana_c_02(ComponentPose=@n1, Alpha=1, ExternalForce="0,0,0") {
    angularDampingOverride=1
    angularSpringConstant=11000
    bAngularSpring=true
    bOverrideAngularDamping=true
    bUseGravityOverride=true
    gravityScale=0.35
    simulationSpace=RootRelative
    boundBone {boneName=dyn_banana_c_02}
    componentAppliedLinearAccClamp=<2000 2000 2000>
    componentLinearAccScale=<0.25 0.14 0.1>
    componentLinearVelScale=<0.25 0.14 0.1>
    gravityOverride {z=-700}
    physicsBodyDefinitions [{
      boundBone {boneName=dyn_banana_c_02}
      boxExtents=<45 72 33>
      constraintSetup {angularLimitsMax=<0 19 2>, angularLimitsMin=<0 -26 -2>, angularTarget=<1 0 0>}
      localJointOffset=<1.6 0 0>
     }]
    simSpaceSettings {maxAngularAcceleration=25, maxAngularVelocity=2.5, simSpaceAngularAlpha=1}
   }
   n3 = Anim Dynamics - Bone: dyn_banana_l_01(ComponentPose=@n2, Alpha=1, ExternalForce="0,0,0") {
    angularDampingOverride=1
    angularSpringConstant=11000
    bAngularSpring=true
    bOverrideAngularDamping=true
    bUseGravityOverride=true
    gravityScale=0.35
    simulationSpace=RootRelative
    boundBone {boneName=dyn_banana_l_01}
    componentAppliedLinearAccClamp=<2000 2000 2000>
    componentLinearAccScale=<0.25 0.14 0.1>
    componentLinearVelScale=<0.25 0.14 0.1>
    gravityOverride {z=-700}
    physicsBodyDefinitions [{
      boundBone {boneName=dyn_banana_l_01}
      boxExtents=<45 72 33>
      constraintSetup {angularLimitsMax=<0 19 2>, angularLimitsMin=<0 -26 -2>, angularTarget=<1 0 0>}
      localJointOffset=<1.6 0 0>
     }]
    simSpaceSettings {maxAngularAcceleration=25, maxAngularVelocity=2.5, simSpaceAngularAlpha=1}
   }
   n4 = Anim Dynamics - Bone: dyn_banana_l_02(ComponentPose=@n3, Alpha=1, ExternalForce="0,0,0") {
    angularDampingOverride=1
    angularSpringConstant=11000
    bAngularSpring=true
    bOverrideAngularDamping=true
    bUseGravityOverride=true
    gravityScale=0.35
    simulationSpace=RootRelative
    boundBone {boneName=dyn_banana_l_02}
    componentAppliedLinearAccClamp=<2000 2000 2000>
    componentLinearAccScale=<0.25 0.14 0.1>
    componentLinearVelScale=<0.25 0.14 0.1>
    gravityOverride {z=-700}
    physicsBodyDefinitions [{
      boundBone {boneName=dyn_banana_l_02}
      boxExtents=<45 72 33>
      constraintSetup {angularLimitsMax=<0 19 2>, angularLimitsMin=<0 -26 -2>, angularTarget=<1 0 0>}
      localJointOffset=<1.6 0 0>
     }]
    simSpaceSettings {maxAngularAcceleration=25, maxAngularVelocity=2.5, simSpaceAngularAlpha=1}
   }
   n5 = Anim Dynamics - Bone: dyn_banana_r_01(ComponentPose=@n4, Alpha=1, ExternalForce="0,0,0") {
    angularDampingOverride=1
    angularSpringConstant=11000
    bAngularSpring=true
    bOverrideAngularDamping=true
    bUseGravityOverride=true
    gravityScale=0.35
    simulationSpace=RootRelative
    boundBone {boneName=dyn_banana_r_01}
    componentAppliedLinearAccClamp=<2000 2000 2000>
    componentLinearAccScale=<0.25 0.14 0.1>
    componentLinearVelScale=<0.25 0.14 0.1>
    gravityOverride {z=-700}
    physicsBodyDefinitions [{
      boundBone {boneName=dyn_banana_r_01}
      boxExtents=<45 72 33>
      constraintSetup {angularLimitsMax=<0 19 2>, angularLimitsMin=<0 -26 -2>, angularTarget=<1 0 0>}
      localJointOffset=<1.6 0 0>
     }]
    simSpaceSettings {maxAngularAcceleration=25, maxAngularVelocity=2.5, simSpaceAngularAlpha=1}
   }
   n6 = Anim Dynamics - Bone: dyn_banana_r_02(ComponentPose=@n5, Alpha=1, ExternalForce="0,0,0") {
    angularDampingOverride=1
    angularSpringConstant=11000
    bAngularSpring=true
    bOverrideAngularDamping=true
    bUseGravityOverride=true
    gravityScale=0.35
    simulationSpace=RootRelative
    boundBone {boneName=dyn_banana_r_02}
    componentAppliedLinearAccClamp=<2000 2000 2000>
    componentLinearAccScale=<0.25 0.14 0.1>
    componentLinearVelScale=<0.25 0.14 0.1>
    gravityOverride {z=-700}
    physicsBodyDefinitions [{
      boundBone {boneName=dyn_banana_r_02}
      boxExtents=<45 72 33>
      constraintSetup {angularLimitsMax=<0 19 2>, angularLimitsMin=<0 -26 -2>, angularTarget=<1 0 0>}
      localJointOffset=<1.6 0 0>
     }]
    simSpaceSettings {maxAngularAcceleration=25, maxAngularVelocity=2.5, simSpaceAngularAlpha=1}
   }
   n7 = Component To Local(ComponentPose=@n6)
   n8 = Anim Dynamics - Bone: dyn_banana_c_01(ComponentPose=Local To Component(LocalPose=Use cached pose 'PX_dairy_01'()), Alpha=1, ExternalForce="0,0,0") {
    angularDampingOverride=1
    angularSpringConstant=10000
    bAngularSpring=true
    bOverrideAngularDamping=true
    bUseGravityOverride=true
    gravityScale=0.35
    simulationSpace=RootRelative
    boundBone {boneName=dyn_banana_c_01}
    componentAppliedLinearAccClamp=<2000 2000 2000>
    componentLinearAccScale=<0.12 0.1 0.1>
    componentLinearVelScale=<0.12 0.1 0.1>
    gravityOverride {z=-700}
    physicsBodyDefinitions [{
      boundBone {boneName=dyn_banana_c_01}
      boxExtents=<44 44 44>
      constraintSetup {angularLimitsMax=<0 19 8.4>, angularLimitsMin=<0 -25 -8.4>, angularTarget=<1 0 0>}
      localJointOffset=<1.35 0 0>
     }]
    simSpaceSettings {maxAngularAcceleration=25, maxAngularVelocity=2.5, simSpaceAngularAlpha=1}
   }
   n9 = Anim Dynamics - Bone: dyn_banana_c_02(ComponentPose=@n8, Alpha=1, ExternalForce="0,0,0") {
    angularDampingOverride=1
    angularSpringConstant=10000
    bAngularSpring=true
    bOverrideAngularDamping=true
    bUseGravityOverride=true
    gravityScale=0.35
    simulationSpace=RootRelative
    boundBone {boneName=dyn_banana_c_02}
    componentAppliedLinearAccClamp=<2000 2000 2000>
    componentLinearAccScale=<0.12 0.1 0.1>
    componentLinearVelScale=<0.12 0.1 0.1>
    gravityOverride {z=-700}
    physicsBodyDefinitions [{
      boundBone {boneName=dyn_banana_c_02}
      boxExtents=<44 44 44>
      constraintSetup {angularLimitsMax=<0 19 8.4>, angularLimitsMin=<0 -25 -8.4>, angularTarget=<1 0 0>}
      localJointOffset=<1.35 0 0>
     }]
    simSpaceSettings {maxAngularAcceleration=25, maxAngularVelocity=2.5, simSpaceAngularAlpha=1}
   }
   n10 = Anim Dynamics - Bone: dyn_banana_l_01(ComponentPose=@n9, Alpha=1, ExternalForce="0,0,0") {
    angularDampingOverride=1
    angularSpringConstant=10000
    bAngularSpring=true
    bOverrideAngularDamping=true
    bUseGravityOverride=true
    gravityScale=0.35
    simulationSpace=RootRelative
    boundBone {boneName=dyn_banana_l_01}
    componentAppliedLinearAccClamp=<2000 2000 2000>
    componentLinearAccScale=<0.12 0.1 0.1>
    componentLinearVelScale=<0.12 0.1 0.1>
    gravityOverride {z=-700}
    physicsBodyDefinitions [{
      boundBone {boneName=dyn_banana_l_01}
      boxExtents=<44 44 44>
      constraintSetup {angularLimitsMax=<0 19 8.4>, angularLimitsMin=<0 -25 -8.4>, angularTarget=<1 0 0>}
      localJointOffset=<1.35 0 0>
     }]
    simSpaceSettings {maxAngularAcceleration=25, maxAngularVelocity=2.5, simSpaceAngularAlpha=1}
   }
   n11 = Anim Dynamics - Bone: dyn_banana_l_02(ComponentPose=@n10, Alpha=1, ExternalForce="0,0,0") {
    angularDampingOverride=1
    angularSpringConstant=10000
    bAngularSpring=true
    bOverrideAngularDamping=true
    bUseGravityOverride=true
    gravityScale=0.35
    simulationSpace=RootRelative
    boundBone {boneName=dyn_banana_l_02}
    componentAppliedLinearAccClamp=<2000 2000 2000>
    componentLinearAccScale=<0.12 0.1 0.1>
    componentLinearVelScale=<0.12 0.1 0.1>
    gravityOverride {z=-700}
    physicsBodyDefinitions [{
      boundBone {boneName=dyn_banana_l_02}
      boxExtents=<44 44 44>
      constraintSetup {angularLimitsMax=<0 19 8.4>, angularLimitsMin=<0 -25 -8.4>, angularTarget=<1 0 0>}
      localJointOffset=<1.35 0 0>
     }]
    simSpaceSettings {maxAngularAcceleration=25, maxAngularVelocity=2.5, simSpaceAngularAlpha=1}
   }
   n12 = Anim Dynamics - Bone: dyn_banana_r_01(ComponentPose=@n11, Alpha=1, ExternalForce="0,0,0") {
    angularDampingOverride=1
    angularSpringConstant=10000
    bAngularSpring=true
    bOverrideAngularDamping=true
    bUseGravityOverride=true
    gravityScale=0.35
    simulationSpace=RootRelative
    boundBone {boneName=dyn_banana_r_01}
    componentAppliedLinearAccClamp=<2000 2000 2000>
    componentLinearAccScale=<0.12 0.1 0.1>
    componentLinearVelScale=<0.12 0.1 0.1>
    gravityOverride {z=-700}
    physicsBodyDefinitions [{
      boundBone {boneName=dyn_banana_r_01}
      boxExtents=<44 44 44>
      constraintSetup {angularLimitsMax=<0 19 8.4>, angularLimitsMin=<0 -25 -8.4>, angularTarget=<1 0 0>}
      localJointOffset=<1.35 0 0>
     }]
    simSpaceSettings {maxAngularAcceleration=25, maxAngularVelocity=2.5, simSpaceAngularAlpha=1}
   }
   n13 = Anim Dynamics - Bone: dyn_banana_r_02(ComponentPose=@n12, Alpha=1, ExternalForce="0,0,0") {
    angularDampingOverride=1
    angularSpringConstant=10000
    bAngularSpring=true
    bOverrideAngularDamping=true
    bUseGravityOverride=true
    gravityScale=0.35
    simulationSpace=RootRelative
    boundBone {boneName=dyn_banana_r_02}
    componentAppliedLinearAccClamp=<2000 2000 2000>
    componentLinearAccScale=<0.12 0.1 0.1>
    componentLinearVelScale=<0.12 0.1 0.1>
    gravityOverride {z=-700}
    physicsBodyDefinitions [{
      boundBone {boneName=dyn_banana_r_02}
      boxExtents=<44 44 44>
      constraintSetup {angularLimitsMax=<0 19 8.4>, angularLimitsMin=<0 -25 -8.4>, angularTarget=<1 0 0>}
      localJointOffset=<1.35 0 0>
     }]
    simSpaceSettings {maxAngularAcceleration=25, maxAngularVelocity=2.5, simSpaceAngularAlpha=1}
   }
   n14 = Component To Local(ComponentPose=@n13)
   n15 = Get bIsInFrontEnd()
   n16 = Anim Dynamics - Bone: dyn_banana_c_01(ComponentPose=Local To Component(LocalPose=Use cached pose 'PX_dairy_01'()), Alpha=1, ExternalForce="0,0,0") {
    angularDampingOverride=1
    angularSpringConstant=11000
    bAngularSpring=true
    bOverrideAngularDamping=true
    bUseGravityOverride=true
    gravityScale=0.35
    simulationSpace=RootRelative
    boundBone {boneName=dyn_banana_c_01}
    componentAppliedLinearAccClamp=<2000 2000 2000>
    componentLinearAccScale=<0.25 0.14 0.1>
    componentLinearVelScale=<0.25 0.14 0.1>
    gravityOverride {z=-700}
    physicsBodyDefinitions [{
      boundBone {boneName=dyn_banana_c_01}
      boxExtents=<45 73 34>
      constraintSetup {angularLimitsMax=<0 19 2>, angularLimitsMin=<0 -26 -2>, angularTarget=<1 0 0>}
      localJointOffset=<1.15 0 0>
     }]
    simSpaceSettings {maxAngularAcceleration=25, maxAngularVelocity=2.5, simSpaceAngularAlpha=1}
   }
   n17 = Anim Dynamics - Bone: dyn_banana_c_02(ComponentPose=@n16, Alpha=1, ExternalForce="0,0,0") {
    angularDampingOverride=1
    angularSpringConstant=11000
    bAngularSpring=true
    bOverrideAngularDamping=true
    bUseGravityOverride=true
    gravityScale=0.35
    simulationSpace=RootRelative
    boundBone {boneName=dyn_banana_c_02}
    componentAppliedLinearAccClamp=<2000 2000 2000>
    componentLinearAccScale=<0.25 0.14 0.1>
    componentLinearVelScale=<0.25 0.14 0.1>
    gravityOverride {z=-700}
    physicsBodyDefinitions [{
      boundBone {boneName=dyn_banana_c_02}
      boxExtents=<45 73 34>
      constraintSetup {angularLimitsMax=<0 19 2>, angularLimitsMin=<0 -26 -2>, angularTarget=<1 0 0>}
      localJointOffset=<1.15 0 0>
     }]
    simSpaceSettings {maxAngularAcceleration=25, maxAngularVelocity=2.5, simSpaceAngularAlpha=1}
   }
   n18 = Anim Dynamics - Bone: dyn_banana_l_01(ComponentPose=@n17, Alpha=1, ExternalForce="0,0,0") {
    angularDampingOverride=1
    angularSpringConstant=11000
    bAngularSpring=true
    bOverrideAngularDamping=true
    bUseGravityOverride=true
    gravityScale=0.35
    simulationSpace=RootRelative
    boundBone {boneName=dyn_banana_l_01}
    componentAppliedLinearAccClamp=<2000 2000 2000>
    componentLinearAccScale=<0.25 0.14 0.1>
    componentLinearVelScale=<0.25 0.14 0.1>
    gravityOverride {z=-700}
    physicsBodyDefinitions [{
      boundBone {boneName=dyn_banana_l_01}
      boxExtents=<45 73 34>
      constraintSetup {angularLimitsMax=<0 19 2>, angularLimitsMin=<0 -26 -2>, angularTarget=<1 0 0>}
      localJointOffset=<1.15 0 0>
     }]
    simSpaceSettings {maxAngularAcceleration=25, maxAngularVelocity=2.5, simSpaceAngularAlpha=1}
   }
   n19 = Anim Dynamics - Bone: dyn_banana_l_02(ComponentPose=@n18, Alpha=1, ExternalForce="0,0,0") {
    angularDampingOverride=1
    angularSpringConstant=11000
    bAngularSpring=true
    bOverrideAngularDamping=true
    bUseGravityOverride=true
    gravityScale=0.35
    simulationSpace=RootRelative
    boundBone {boneName=dyn_banana_l_02}
    componentAppliedLinearAccClamp=<2000 2000 2000>
    componentLinearAccScale=<0.25 0.14 0.1>
    componentLinearVelScale=<0.25 0.14 0.1>
    gravityOverride {z=-700}
    physicsBodyDefinitions [{
      boundBone {boneName=dyn_banana_l_02}
      boxExtents=<45 73 34>
      constraintSetup {angularLimitsMax=<0 19 2>, angularLimitsMin=<0 -26 -2>, angularTarget=<1 0 0>}
      localJointOffset=<1.15 0 0>
     }]
    simSpaceSettings {maxAngularAcceleration=25, maxAngularVelocity=2.5, simSpaceAngularAlpha=1}
   }
   n20 = Anim Dynamics - Bone: dyn_banana_r_01(ComponentPose=@n19, Alpha=1, ExternalForce="0,0,0") {
    angularDampingOverride=1
    angularSpringConstant=11000
    bAngularSpring=true
    bOverrideAngularDamping=true
    bUseGravityOverride=true
    gravityScale=0.35
    simulationSpace=RootRelative
    boundBone {boneName=dyn_banana_r_01}
    componentAppliedLinearAccClamp=<2000 2000 2000>
    componentLinearAccScale=<0.25 0.14 0.1>
    componentLinearVelScale=<0.25 0.14 0.1>
    gravityOverride {z=-700}
    physicsBodyDefinitions [{
      boundBone {boneName=dyn_banana_r_01}
      boxExtents=<45 73 34>
      constraintSetup {angularLimitsMax=<0 19 2>, angularLimitsMin=<0 -26 -2>, angularTarget=<1 0 0>}
      localJointOffset=<1.15 0 0>
     }]
    simSpaceSettings {maxAngularAcceleration=25, maxAngularVelocity=2.5, simSpaceAngularAlpha=1}
   }
   n21 = Anim Dynamics - Bone: dyn_banana_r_02(ComponentPose=@n20, Alpha=1, ExternalForce="0,0,0") {
    angularDampingOverride=1
    angularSpringConstant=11000
    bAngularSpring=true
    bOverrideAngularDamping=true
    bUseGravityOverride=true
    gravityScale=0.35
    simulationSpace=RootRelative
    boundBone {boneName=dyn_banana_r_02}
    componentAppliedLinearAccClamp=<2000 2000 2000>
    componentLinearAccScale=<0.25 0.14 0.1>
    componentLinearVelScale=<0.25 0.14 0.1>
    gravityOverride {z=-700}
    physicsBodyDefinitions [{
      boundBone {boneName=dyn_banana_r_02}
      boxExtents=<45 73 34>
      constraintSetup {angularLimitsMax=<0 19 2>, angularLimitsMin=<0 -26 -2>, angularTarget=<1 0 0>}
      localJointOffset=<1.15 0 0>
     }]
    simSpaceSettings {maxAngularAcceleration=25, maxAngularVelocity=2.5, simSpaceAngularAlpha=1}
   }
   Output Pose(Result=Blend Poses by bool(
     BlendPose_0=Component To Local(ComponentPose=@n21)
     BlendPose_1=Blend Poses by bool(BlendPose_0=@n14, BlendPose_1=@n7, bActiveValue=@n15)
     bActiveValue=Get bIsOnPawn() -> (bIsOnPawn:bool)
    )) {layerGroup=None}
   AnimGraphNode_SubInput_0 = Input Pose()
   Save cached pose 'PX_dairy_01'(Pose=@AnimGraphNode_SubInput_0) {cacheName=PX_dairy_01}
  ]}
 EventGraph {nodes [
   TryGetPawnOwner()
   Event BlueprintUpdateAnimation()
  ]}
]
variables []
